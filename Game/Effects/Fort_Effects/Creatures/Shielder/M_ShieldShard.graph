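# Translucent unlit particle material for Shielder creature shield shards, tinted by particle color through an emissive adjustment function.
# A sub-UV shard sheet supplies two channels: G, broken up by panning noise, blends to R by each pixel's position along the direction away from the actor.
# That directional term optionally scales the mask, which is multiplied by particle alpha and Fresnel.
# Mobile shading uses a constant blend without directional scaling, and Low quality drops Fresnel.
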Material /Game/Effects/Fort_Effects/Creatures/Shielder/M_ShieldShard
graph {
 properties {
  bUsedWithMeshParticles=true
  bUsedWithNanite=true
  bUsedWithNiagaraMeshParticles=true
  bUsedWithNiagaraSprites=true
  bUsedWithParticleSprites=true
  blendMode=BLEND_Translucent
  pixelDepthOffsetMode=PDOM_Legacy
  shadingModel=MSM_Unlit
  twoSided=true
  usedShadingModels=Unlit
 }
 nodes [
  n1 = ParticleColor()
  MaterialFunctionCall(Emissive=@n1.RGB) {materialFunction=&MF_FX_EmissiveAdjustments}
  n2 = ParticleSubUV!(UVs, Tex, Apply View MipBias) {texture=&/T_ShielderShardSheet}
  n3 = TextureCoordinate()
  n4 = Add(A=TextureSample(UVs=Panner(Coordinate=@n3) {speedX=1.5, speedY=3}) {texture=&T_NoisePack_01}.R) {constB=0.35}
  n5 = Subtract(A=WorldPosition().XYZ, B=ParticlePositionWS())
  n6 = ScalarParameter() {defaultValue=64, parameterName=ParticleSize}
  n7 = ParticlePositionWS()
  n8 = ActorPositionWS()
  n9 = DotProduct(A=Add(A=Divide(A=@n5, B=@n6) {constB=1}) {constB=0.1}, B=Normalize(VectorInput=Subtract(A=@n7, B=@n8)))
  n10 = OneMinus(None=@n9)
  n11 = Saturate(None=OneMinus(None=Power(Base=@n10)))
  n12 = LinearInterpolate(
   A=Multiply(A=@n2.G, B=Saturate(None=@n4))
   B=@n2.R
   Alpha=ShadingPathSwitch(Default=@n11, Deferred=@n11, Mobile=Constant() {r=0.5}) {desc=Migrated from FeatureLevelSwitch}
  ) {constAlpha=0}
  n13 = Constant() {r=1}
  n14 = Multiply(A=@n12, B=StaticSwitchParameter(True=@n9, False=@n13) {defaultValue=true, parameterName=Use away})
  n15 = MaterialFunctionCall(Exponent=@n13) {materialFunction=&MF_Fresnel_FX}
  Multiply(
   A=Multiply(A=ShadingPathSwitch(Default=@n14, Deferred=@n14, Mobile=@n12) {desc=Migrated from FeatureLevelSwitch}, B=@n1.A)
   B=QualitySwitch(Default=@n15.Fresnel, Low=@n13, High=@n15.Fresnel, Medium=@n15.Fresnel)
  )
  Reroute(None=@n9)
 ]
}
