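# Unlit translucent particle material for the Shielder creature's fracture effect, usable on mesh particles.
# Particle color feeds an emissive adjustment function.
# Particle alpha is multiplied by tiled shard texture and noise, depth fade, and an inverted Fresnel that changes over particle lifetime.
# The shard texture also scales an offset along vertex normals, with Dynamic Parameters controlling texture, noise and offset strength.
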Material /Game/Effects/Fort_Effects/Creatures/Shielder/M_ShielderFracture
graph {
 properties {
  bUsedWithMeshParticles=true
  bUsedWithNanite=true
  bUsedWithNiagaraMeshParticles=true
  blendMode=BLEND_Translucent
  pixelDepthOffsetMode=PDOM_Legacy
  shadingModel=MSM_Unlit
  usedShadingModels=Unlit
 }
 nodes [
  n1 = ParticleColor()
  MaterialFunctionCall(Emissive=@n1.RGB) {materialFunction=&MF_FX_EmissiveAdjustments}
  n2 = LinearInterpolate(
   B=TextureSample(UVs=TextureCoordinate() {uTiling=5, vTiling=3}) {texture=&T_NoisePack_01}.R
   Alpha=DynamicParameter() {paramNames [TextureStrength, OffsetStrength, NoiseStrength, Param4]}.NoiseStrength
  ) {constA=1}
  n3 = TextureSample(UVs=TextureCoordinate() {uTiling=5, vTiling=3}) {texture=&/T_ShieldShard}
  n4 = LinearInterpolate(
   B=Multiply(A=Power(Base=@n2), B=@n3.R)
   Alpha=DynamicParameter() {paramNames [TextureStrength, OffsetStrength, NoiseStrength, Param4]}.TextureStrength
  ) {constA=1}
  n5 = ParticleRelativeTime()
  n6 = Add(
   A=MaterialFunctionCall(Exponent=LinearInterpolate(Alpha=@n5) {constA=2, constAlpha=1, constB=6}) {materialFunction=&MF_Fresnel_FX}.Fresnel Inverted
   B=LinearInterpolate(Alpha=@n5) {constA=0.08, constB=0}
  ) {constB=0.08}
  Multiply(A=Multiply(A=Multiply(A=@n1.A, B=@n4), B=Saturate(None=@n6)), B=DepthFade(FadeDistance=Constant() {r=12}))
  Multiply(
   A=Multiply(A=@n3.R, B=DynamicParameter() {paramNames [TextureStrength, OffsetStrength, NoiseStrength, Param4]}.OffsetStrength) {constB=8}
   B=VertexNormalWS()
  )
 ]
}
